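# Niagara system for the start burst of the FantasySparkle aura cosmetic, played on a character's skeletal mesh.
# One-shot Flare, Dark_Glow and Corona sprites flash alongside a brief curl-noise spark burst.
# A burst of sprite-and-mesh Butterflies drifts out from a bone, tinted from a color array, and Trail sparkles follow them via a particle attribute reader.
# User flags scale spread for large bodies and dim glow in the frontend, with time-of-day scaling.
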NiagaraSystem /SparksCosmetics/Items/Auras/Assets/NS_Aura_FantasySparkle/NS_Aura_FantasySparkle_Start
dependencies {
 usedDataInterfaces [
  &NiagaraDataInterfaceCurve
  &NiagaraDataInterfaceColorCurve
  &NiagaraDataInterfaceParticleRead
  &NiagaraDataInterfaceArrayColor
 ]
 usedDynamicInputs []
 usedModules [
  &self:SystemSpawnScript.SystemScriptSource.SystemScriptGraph.NiagaraNodeAssignment_0.SetVariables_#1
  &self:SystemSpawnScript.SystemScriptSource.SystemScriptGraph.NiagaraNodeAssignment_1.SetVariables_#2
  &SystemState
 ]
}
emitters {
 Butterflies {
  emitterUpdateScript [
   &EmitterState()
   &SpawnBurst_Instantaneous(Spawn Count=8, Use Spawn Probability=false, UseLoopCountLimit=true)
  ]
  particleSpawnScript [
   &InitializeParticle(
    Color=<rgba 1 0.7275 0.4302 1>
    Lifetime Min=0.5
    Lifetime Mode=::NewEnumerator1
    Mesh Scale=<6 3 4>
    Mesh Scale Mode=::NewEnumerator4
    Sprite Size Mode=::NewEnumerator3
    Uniform Sprite Size=20
    UsePositionOffset=false
    Write Material Random=true
   )
   &ShapeLocation(
    MaskSpawnGroup=false
    Non Uniform Scale=&MakeCustomVectorFromBool(Bool=$User.IsLargeBody, True Vector=<0.5 1.8 1>, False Vector=<1 0.6 0.8>)
    Override Seed=false
    Sphere Radius=15
    UseOwnerScale=true
   )
   &SkeletalMeshLocation(
    Bone / Socket Index=7
    Bone Sampling Mode=::NewEnumerator4
    MaskSpawnGroup=false
    Orientation Sampling=::NewEnumerator1
    Sampled Position Offset=$Particles.ShapeLocation.ShapeVector
    Skeletal Mesh=$System.Skeletal Mesh
   )
   &AddVelocity(
    Module.Use Falloff Distance=false
    Override Seed=false
    Velocity Mode=::NewEnumerator1
    Velocity Origin=$Particles.SkeletalMeshLocation.SampledPosition
    Velocity Speed=&MakeCustomFloatFromBool(Bool=$User.IsLargeBody, True Float=150, False Float=50)
   )
   &AddVelocity(Velocity=<0 -20 10>) {module=AddVelocity001}
   /SetVariables(Particles.ScaleVariance:NiagaraFloat=&RandomRangeFloat(Minimum=0.5, Maximum=1)) {isSetParameters=true}
   /SetVariables(Particles.glowColor:LinearColor=&MultiplyLinearColor(
     Linear Color=&SelectLinearColorFromArray(Array Sampling Mode=::NewEnumerator0, Direct Array Index=?, Color Selection Array=$System.ColorRange, Random Seed=0)
     Scale Factor=&MakeVector4_FromFloat(Float=&MakeCustomFloatFromBool(
       Bool=&Boolean_OrOperation(Bool1=$User.IsFrontEnd, Bool2=&Boolean_OrOperation(Bool1=$User.bIsFrontEnd, Bool2=$User.bIsFrontEndPreview))
       True Float=0.5
       False Float=1
      ))
    )) {isSetParameters=true}
   &CameraOffset(Camera Offset Amount=-50)
  ]
  particleUpdateScript [
   &ParticleState()
   &ScaleMeshSize(Scale Factor=&VectorFromFloat(Value=&FloatFromCurve(
      FloatCurve=DataInterface() {propertyValues {Curve {keys [
          {arriveTangent=14.2603, interpMode=RCIM_Cubic, leaveTangent=14.2603, tangentMode=RCTM_User, time=0, value=0}
          {arriveTangentWeight=0.08, interpMode=RCIM_Cubic, leaveTangentWeight=0.3468, tangentMode=RCTM_User, time=0.15, value=1}
          {arriveTangentWeight=0.3345, interpMode=RCIM_Cubic, leaveTangentWeight=0.1029, tangentMode=RCTM_User, time=0.6, value=1}
          {arriveTangent=-3.4582, tangentMode=RCTM_User, time=1, value=0}
         ]}}}
      Scale Curve=$Particles.ScaleVariance
     )))
   &ScaleSpriteSize(
    Uniform Curve Scale=$Particles.ScaleVariance
    Uniform Curve Sprite Scale=DataInterface() {propertyValues {Curve {keys [
        {arriveTangent=14.2603, interpMode=RCIM_Cubic, leaveTangent=14.2603, tangentMode=RCTM_User, time=0, value=0}
        {arriveTangentWeight=0.08, interpMode=RCIM_Cubic, leaveTangentWeight=0.3468, tangentMode=RCTM_User, time=0.15, value=1}
        {arriveTangentWeight=0.3345, interpMode=RCIM_Cubic, leaveTangentWeight=0.1029, tangentMode=RCTM_User, time=0.6, value=1}
        {arriveTangent=-3.4582, tangentMode=RCTM_User, time=1, value=0}
       ]}}}
   )
   &SkeletalMeshLocation(
    Bone / Socket Index=7
    Bone Sampling Mode=::NewEnumerator4
    MaskSpawnGroup=false
    Orientation Sampling=::NewEnumerator1
    Position Sampling=::NewEnumerator2
    Skeletal Mesh=$System.Skeletal Mesh
   ) {module=SkeletalMeshLocation001}
   /SetVariables(Particles.SourceDistance:NiagaraFloat=&Multiply_Float(
     A=0.02
     B=/Niagara/DynamicInputs/TypeConversions/VectorLengthSafe(Vector:Vector3f=&Subtract_Vector(
       A=Linked() {linkedVariable {name=Particles.SkeletalMeshLocation001.SampledPosition, type=NiagaraPosition}}
       B=Linked() {linkedVariable {name=Particles.Position, type=NiagaraPosition}}
      ))
    )) {isSetParameters=true}
   &ScaleColor(
    Scale RGB=&VectorFromFloat(Value=&FloatFromCurve(
      FloatCurve=DataInterface() {propertyValues {Curve {keys [<at 0 3 RCIM_Cubic>,<at 1 1>]}}}
      Scale Curve=&Multiply_Float(
       A=&FNTimeOfDayScalar(Night Scale=0.5, Morning Scale=0.8, Evening Scale=0.8)
       B=&MakeCustomFloatFromBool(
        Bool=&Boolean_OrOperation(Bool1=$User.IsFrontEnd, Bool2=&Boolean_OrOperation(Bool1=$User.bIsFrontEnd, Bool2=$User.bIsFrontEndPreview))
        True Float=0.5
        False Float=1
       )
      )
     ))
    ScaleA=false
    ScaleRGB=true
   )
   /SetVariables(
    Particles.GlowDynamic:Vector4f=&MakeVector4(
     X=1
     Y=1
     Z=0
     W=&FloatFromCurve(
      FloatCurve=DataInterface() {propertyValues {Curve {keys [
          {arriveTangent=14.2603, interpMode=RCIM_Cubic, leaveTangent=14.2603, tangentMode=RCTM_User, time=0, value=0}
          {arriveTangentWeight=0.08, interpMode=RCIM_Cubic, leaveTangentWeight=0.3468, tangentMode=RCTM_User, time=0.15, value=1}
          {arriveTangentWeight=0.3345, interpMode=RCIM_Cubic, leaveTangentWeight=0.1029, tangentMode=RCTM_User, time=0.6, value=1}
          {arriveTangent=-3.4582, tangentMode=RCTM_User, time=1, value=0}
         ]}}}
      Scale Curve=$Particles.ScaleVariance
     )
    )
    Particles.glowColor:LinearColor=/Niagara/DynamicInputs/Lerp/SmoothLerpOverTime_Color(
     Smooth Value:LinearColor=$Particles.glowColor
     Target Value:LinearColor=&MultiplyLinearColor(
      Linear Color=&SelectLinearColorFromArray(Array Sampling Mode=::NewEnumerator0, Direct Array Index=?, Color Selection Array=$System.ColorRange, Random Seed=0)
      Scale Factor=&MakeVector4_FromFloat(Float=&Multiply_Float(
        A=&FNTimeOfDayScalar(Night Scale=0.3, Morning Scale=0.4, Evening Scale=0.4)
        B=&MakeCustomFloatFromBool(
         Bool=&Boolean_OrOperation(Bool1=$User.IsFrontEnd, Bool2=&Boolean_OrOperation(Bool1=$User.bIsFrontEnd, Bool2=$User.bIsFrontEndPreview))
         True Float=0.5
         False Float=1
        )
       ))
     )
     Convergence Rate:NiagaraFloat=1
    )
   ) {isSetParameters=true}
   /Niagara/Modules/Update/Forces/CurlNoiseForce(
    Noise Frequency:NiagaraFloat=&RandomRangeFloat(Minimum=0.05, Maximum=0.25)
    Noise Strength=&MakeCustomFloatFromBool(Bool=$User.IsLargeBody, True Float=100, False Float=50)
    Pan Noise=true
    Pan Noise Field=<0 0 1>
   )
   &Drag(
    Drag=&Lerp_Float(A=0, B=1, Alpha=&ClampFloat(Float=$Particles.SourceDistance))
    Use Linear Drag=true
    Use Rotational Drag=true
   )
   &SolveForcesAndVelocity(Clamp Velocity=false, Limit Acceleration=false)
   &DynamicMaterialParameters(
    Index 0 Param 1=&RandomRangeFloat(Minimum=1.1, Maximum=2)
    Index 0 Param 2=&RandomRangeFloat(Minimum=1, Maximum=2)
    Index 0 Param 3=&RandomRangeFloat(Minimum=1, Maximum=2)
    Index 1 Param 1=&FloatFromCurve(
     FloatCurve=DataInterface() {propertyValues {Curve {
        defaultValue=1
        keys [
         {interpMode=RCIM_Cubic, tangentMode=RCTM_User, time=0, value=1.3}
         {interpMode=RCIM_Cubic, tangentMode=RCTM_User, time=0.5, value=0}
         {tangentMode=RCTM_User, time=1, value=1.3}
        ]
       }}}
     Scale Curve=0.1
    )
    Index 1 Param 2=&RandomRangeFloat(Minimum=0, Maximum=1)
    Module.Param0WriteEnabled=true
    Module.Param0WriteEnabled1=true
    Module.Param1WriteEnabled=true
    Module.Param1WriteEnabled1=true
    Module.Param2WriteEnabled=true
    Module.Param2WriteEnabled1=true
    Module.Param3WriteEnabled=true
    Module.Param3WriteEnabled1=true
    Write Parameter Index 1=true
   )
  ]
  rendererClasses [&NiagaraSpriteRendererProperties, &NiagaraMeshRendererProperties]
  renderers [
   {rendererClass=&NiagaraSpriteRendererProperties}
   {rendererClass=&NiagaraMeshRendererProperties, rendererIndex=1}
   {rendererClass=&NiagaraMeshRendererProperties, rendererIndex=2}
  ]
 }
 Corona {
  emitterUpdateScript [
   &EmitterState()
   &SpawnBurst_Instantaneous(Spawn Count=1, Use Spawn Probability=false)
  ]
  particleSpawnScript [
   &InitializeParticle(
    Color=<rgba 2 2 2 1>
    Lifetime=0.25
    Sprite Rotation Mode=::NewEnumerator2
    Sprite Size Mode=::NewEnumerator3
    Uniform Sprite Size=400
    UsePositionOffset=false
    Write Material Random=true
   )
   &DynamicMaterialParameters(
    Index 0 Param 1=&RandomRangeFloat(Minimum=0.4, Maximum=0.6)
    Index 0 Param 2=1
    Index 0 Param 3=1
    Index 0 Param 4=1
    Module.Param0WriteEnabled=true
    Module.Param1WriteEnabled=true
    Module.Param2WriteEnabled=true
    Module.Param3WriteEnabled=true
   )
   &SystemLocation(MaskSpawnGroup=false, Offset=<0 0 120>)
  ]
  particleUpdateScript [
   &ParticleState()
   &SolveForcesAndVelocity(Clamp Velocity=false, Limit Acceleration=false)
   &SkeletalMeshLocation(Bone Sampling Mode=::NewEnumerator4, MaskSpawnGroup=false, Skeletal Mesh=$System.Skeletal Mesh) {module=SkeletalMeshLocation001}
   &ScaleColor(
    Scale Alpha=&Multiply_Float(
     A=&FNTimeOfDayScalar(Night Scale=1, Morning Scale=1, Evening Scale=1)
     B=&FloatFromCurve(FloatCurve=DataInterface() {propertyValues {Curve {keys [
          {arriveTangent=9.6062, interpMode=RCIM_Cubic, leaveTangent=9.6062, tangentMode=RCTM_User, time=0, value=0}
          {interpMode=RCIM_Cubic, tangentMode=RCTM_User, time=0.2, value=1}
          {interpMode=RCIM_Cubic, tangentMode=RCTM_User, time=1, value=0}
         ]}}})
    )
    Scale RGB=&MakeVectorFromLinearColorRGB(Color=&ColorFromCurve(ColorCurve=DataInterface() {propertyValues {
        AlphaCurve {keys [<at 0 1 RCIM_Cubic>]}
        BlueCurve {keys [<at 0.2018 1 RCIM_Cubic>,<at 0.7606 1>]}
        GreenCurve {keys [<at 0.2018 1 RCIM_Cubic>,<at 0.7606 0.0435>]}
        RedCurve {keys [<at 0.2018 1 RCIM_Cubic>,<at 0.7606 0.629>]}
       }}))
    ScaleA=true
    ScaleRGB=true
   )
   &ScaleSpriteSize(Scale Factor=&Vector2DFromFloat(Value=&FloatFromCurve(FloatCurve=DataInterface() {propertyValues {Curve {keys [
          {arriveTangent=2.1722, interpMode=RCIM_Cubic, leaveTangent=2.1722, tangentMode=RCTM_User, time=0, value=0}
          {arriveTangent=0.0224, interpMode=RCIM_Cubic, leaveTangent=0.0224, tangentMode=RCTM_User, time=1, value=1}
         ]}}})))
   &CameraOffset(Camera Offset Amount=-30)
   &DynamicMaterialParameters(
    Index 0 Param 1=0.75
    Index 0 Param 2=2
    Index 0 Param 3=1
    Module.Param0WriteEnabled=true
    Module.Param1WriteEnabled=true
    Module.Param2WriteEnabled=true
    Module.Param3WriteEnabled=true
   ) {module=DynamicMaterialParameters001}
  ]
  rendererClasses [&NiagaraSpriteRendererProperties]
  renderers [{rendererClass=&NiagaraSpriteRendererProperties}]
 }
 Dark_Glow {
  emitterUpdateScript [
   &EmitterState()
   &SpawnBurst_Instantaneous(Spawn Count=1, Use Spawn Probability=false)
  ]
  particleSpawnScript [&InitializeParticle(
    Color=<rgba 0.2 0 0.1357 1>
    Lifetime=0.4
    Position Offset=<0 0 100>
    Sprite Rotation Mode=::NewEnumerator2
    Sprite Size Mode=::NewEnumerator3
    Uniform Sprite Size=230
    UsePositionOffset=true
    Write Material Random=true
   )]
  particleUpdateScript [
   &ParticleState()
   &SkeletalMeshLocation(Bone Sampling Mode=::NewEnumerator4, MaskSpawnGroup=false, Skeletal Mesh=$System.Skeletal Mesh) {module=SkeletalMeshLocation001}
   &ScaleColor(
    Scale Alpha=&FloatFromCurve(
     FloatCurve=DataInterface() {propertyValues {Curve {keys [
         {interpMode=RCIM_Cubic, tangentMode=RCTM_User, time=0, value=0}
         {interpMode=RCIM_Cubic, tangentMode=RCTM_User, time=0.2, value=1}
         {interpMode=RCIM_Cubic, tangentMode=RCTM_User, time=0.5, value=1}
         {tangentMode=RCTM_User, time=1, value=0}
        ]}}}
     Scale Curve=2
    )
    ScaleA=true
    ScaleRGB=true
   )
  ]
  rendererClasses [&NiagaraSpriteRendererProperties]
  renderers [{rendererClass=&NiagaraSpriteRendererProperties}]
 }
 Flare {
  emitterUpdateScript [
   &EmitterState()
   &SpawnBurst_Instantaneous(Spawn Count=1, Use Spawn Probability=false)
  ]
  particleSpawnScript [
   &InitializeParticle(
    Color=<rgba 2 0.5772 0.1532 1>
    Lifetime=0.3
    Sprite Rotation Mode=::NewEnumerator2
    Sprite Size Mode=::NewEnumerator3
    Uniform Sprite Size=600
    UsePositionOffset=false
    Write Material Random=true
   )
   &DynamicMaterialParameters(
    Index 0 Param 1=&RandomRangeFloat(Minimum=0.4, Maximum=0.6)
    Index 0 Param 2=1
    Index 0 Param 3=1
    Index 0 Param 4=1
    Module.Param0WriteEnabled=true
    Module.Param1WriteEnabled=true
    Module.Param2WriteEnabled=true
    Module.Param3WriteEnabled=true
   )
   &SystemLocation(MaskSpawnGroup=false, Offset=<0 0 120>)
   /SetVariables(Particles.Sharp Flare Camera Offset=40) {isSetParameters=true}
   &CameraOffset(Camera Offset Amount=-30) {module=CameraOffset001}
  ]
  particleUpdateScript [
   &ParticleState()
   &SolveForcesAndVelocity(Clamp Velocity=false, Limit Acceleration=false)
   &SkeletalMeshLocation(Bone Sampling Mode=::NewEnumerator4, MaskSpawnGroup=false, Skeletal Mesh=$System.Skeletal Mesh) {module=SkeletalMeshLocation001}
   &ScaleColor(
    Scale Alpha=&Multiply_Float(
     A=&FNTimeOfDayScalar(Night Scale=1, Morning Scale=1, Evening Scale=1)
     B=&FloatFromCurve(FloatCurve=DataInterface() {propertyValues {Curve {keys [<at 0 0>,<at 0.2 1>,<at 0.8 1>,<at 1 0>]}}})
    )
    Scale RGB=&Multiply_VectorByFloat(Vector=<1 1 1>, Float=1)
    ScaleA=true
    ScaleRGB=true
   )
   &ScaleSpriteSize(Scale Factor=&Vector2DFromFloat(Value=&FloatFromCurve(FloatCurve=DataInterface() {propertyValues {Curve {keys [<at 0 1.5 RCIM_Cubic>,<at 1 0>]}}})))
   /SetVariables(Particles.Flash color:LinearColor=&MakeLinearColorFromVectorAndFloat(
     "Vector (RGB)"=&MakeVectorFromLinearColorRGB(Color=<rgba 0.5235 0.3277 1 1>)
     "Float (Alpha)"=&Multiply_Float(
      A=&Multiply_Float(
       A=&FNTimeOfDayScalar(Night Scale=1, Morning Scale=1, Evening Scale=1)
       B=&FloatFromCurve(FloatCurve=DataInterface() {propertyValues {Curve {keys [<at 0 0>,<at 0.124 1>,<at 0.496 1>,<at 0.62 0>]}}})
      )
      B=1
     )
    )) {isSetParameters=true}
   /SetVariables(Particles.Flash size:Vector2f=&Multiply_Vector2D_ByFloat(
     Float=&Multiply_Float(
      A=400
      B=&FloatFromCurve(FloatCurve=DataInterface() {propertyValues {Curve {keys [
           {arriveTangent=4.1785, interpMode=RCIM_Cubic, leaveTangent=4.1785, tangentMode=RCTM_User, time=0, value=0}
           {arriveTangent=-0.15, interpMode=RCIM_Cubic, leaveTangent=-0.15, tangentMode=RCTM_User, time=0.1378, value=1}
           {arriveTangent=-0.0405, interpMode=RCIM_Cubic, leaveTangent=-0.0405, tangentMode=RCTM_User, time=0.62, value=0.25}
          ]}}})
     )
     Vector2D=<1 1>
    )) {isSetParameters=true}
  ]
  rendererClasses [&NiagaraSpriteRendererProperties]
  renderers [
   {rendererClass=&NiagaraSpriteRendererProperties}
   {rendererClass=&NiagaraSpriteRendererProperties, rendererIndex=1}
  ]
 }
 Sparks_Burst001 {
  emitterUpdateScript [
   &EmitterState()
   &SpawnRate(
    Spawn Probability=&FloatFromCurve(FloatCurve=DataInterface() {propertyValues {Curve {keys [<at 0 1 RCIM_Constant>,<at 0.075 0>]}}}, CurveIndex=$Emitter.NormalizedLoopAge)
    SpawnRate=250
    Use Spawn Probability=true
   )
  ]
  particleSpawnScript [
   &InitializeParticle(
    Color=&Multiply_LinearColorByFloat(
     Linear Color=&SelectLinearColorFromArray(Array Sampling Mode=::NewEnumerator0, Direct Array Index=?, Color Selection Array=$System.ColorRange, Random Seed=0)
     Float=&RandomRangeFloat(Minimum=1, Maximum=3)
    )
    Lifetime Max=1.75
    Lifetime Mode=::NewEnumerator1
    Position Mode=::NewEnumerator1
    Sprite Rotation Mode=::NewEnumerator2
    Sprite Size=&Lerp_Vector2(
     A=<1 3>
     B=<2 5>
     Alpha=&FloatFromCurve(
      FloatCurve=DataInterface() {propertyValues {Curve {keys [
          {interpMode=RCIM_Cubic, tangentMode=RCTM_User, time=0, value=0}
          {interpMode=RCIM_Cubic, tangentMode=RCTM_User, time=0.3, value=0}
          {interpMode=RCIM_Cubic, tangentMode=RCTM_User, time=0.8, value=1}
         ]}}}
      CurveIndex=&RandomRangeFloat(Minimum=0, Maximum=1)
     )
    )
    Sprite Size Mode=::NewEnumerator4
    UsePositionOffset=false
    Write Material Random=true
   ) {module=InitializeParticle001}
   &ShapeLocation(
    Apply To Particle Position=::NewEnumerator1
    MaskSpawnGroup=false
    Override Seed=false
    Shape Origin=&VectorFromFloat(Value=0)
    Sphere Radius=10
    UseOwnerScale=true
   )
   &AddVelocity(
    Velocity=&MakeVector(
     X=&RandomRangeFloat(Minimum=-400, Maximum=400)
     Y=&RandomRangeFloat(Minimum=-400, Maximum=400)
     Z=&RandomRangeFloat(Minimum=-400, Maximum=400)
    )
    Velocity Speed Scale=3
   ) {module=AddVelocity001}
  ]
  particleUpdateScript [
   &ParticleState()
   &ScaleColor(
    Scale Alpha=&Lerp_Float(
     A=1
     B=0
     Alpha=&FloatFromCurve(FloatCurve=DataInterface() {propertyValues {Curve {keys [
          {arriveTangent=3.1144, interpMode=RCIM_Cubic, leaveTangent=3.1144, tangentMode=RCTM_User, time=0, value=0}
          {interpMode=RCIM_Cubic, tangentMode=RCTM_User, time=1, value=1}
         ]}}})
    )
    Scale RGB=&Multiply_VectorByFloat(
     Vector=<1 1 1>
     Float=&FloatFromCurve(
      FloatCurve=DataInterface() {propertyValues {Curve {keys [
          {arriveTangent=-12.6044, interpMode=RCIM_Cubic, leaveTangent=-12.6044, tangentMode=RCTM_User, time=0, value=5}
          {interpMode=RCIM_Cubic, tangentMode=RCTM_User, time=1, value=1}
         ]}}}
      Scale Curve=10
     )
    )
    ScaleA=true
    ScaleRGB=true
   )
   &ScaleSpriteSize(Uniform Curve Sprite Scale=DataInterface() {propertyValues {Curve {keys [
        {arriveTangent=0.0565, interpMode=RCIM_Cubic, leaveTangent=0.0565, tangentMode=RCTM_User, time=0, value=1}
        {arriveTangent=-0.025, interpMode=RCIM_Cubic, leaveTangent=-0.025, tangentMode=RCTM_User, time=1, value=0}
       ]}}}) {module=ScaleSpriteSize001}
   &SkeletalMeshLocation(
    Bone / Socket Index=$Particles.SpawnGroup
    Bone Sampling Mode=::NewEnumerator4
    MaskSpawnGroup=false
    Position Sampling=::NewEnumerator1
    Sampled Position Offset=$StackContext.ShapeLocation.ShapePosition
    Skeletal Mesh=$System.Skeletal Mesh
    Source Movement Scale:Vector3f=<0 0 0>
   ) {module=SkeletalMeshLocation001}
   &Drag(Drag=10, Use Linear Drag=true, Use Rotational Drag=false)
   /Niagara/Modules/Update/Forces/V2/CurlNoiseForce(Noise Frequency=15, Noise Strength=500, Pan Noise=true, Pan Noise Field=<0.25 0.25 0.25>)
   &SolveForcesAndVelocity(Clamp Velocity=false, Limit Acceleration=false)
  ]
  rendererClasses [&NiagaraSpriteRendererProperties]
  renderers [{rendererClass=&NiagaraSpriteRendererProperties}]
 }
 Trail {
  emitterUpdateScript [
   &EmitterState()
   &SpawnRate(SpawnRate=50, Use Spawn Probability=false)
   /SetVariables(Emitter.Particle Attribute Reader=DataInterface() {propertyValues {EmitterBinding {bindingMode=Other, emitterName=Butterflies}}}) {isSetParameters=true}
  ]
  particleSpawnScript [
   /SetVariables(Particles.ParentPicker:NiagaraInt32=&RandomRangeInt(Minimum=0, Maximum=100)) {isSetParameters=true}
   &self:GetParentInfo(
    Hue Adjustment:NiagaraFloat=&RandomRangeFloat(Minimum=-0.1, Maximum=0.1)
    Parent Attribute Reader:NiagaraDataInterfaceParticleRead=$Emitter.Particle Attribute Reader
    ParentID:NiagaraInt32=$Particles.ParentPicker
    Value Adjustment=1
   ) {module=GetParentInfo}
   /SetVariables(Particles.SparkleOffset:Vector3f=&RandomVector(Vector Scale=1)) {isSetParameters=true}
   &InitializeParticle(
    Color=$Particles.InheritedColor
    Lifetime Max=1
    Lifetime Min=0.4
    Lifetime Mode=::NewEnumerator1
    Position=$Particles.ParentPosition
    Position Mode=::NewEnumerator1
    Position Offset=$Particles.SparkleOffset
    Sprite Size Mode=::NewEnumerator3
    Uniform Sprite Size=&Multiply_Float(A=$Particles.ParentScale, B=1)
    UsePositionOffset=true
    Write Material Random=true
   )
   &InheritVelocity(Inherited Velocity Amount Scale=<0.5 0.5 0.5>, Speed Limit=true, UseSpeedThreshold=false, Velocity Source=$Particles.ParentVelocity)
   &CameraOffset(Camera Offset Amount=-50)
  ]
  particleUpdateScript [
   &ParticleState()
   &SpriteRotationRate(Rotation Rate=&RandomRangeFloat(Minimum=-300, Maximum=300))
   &ScaleSpriteSize(
    Scale Sprite Size Mode=::NewEnumerator0
    Uniform Scale Factor=&FloatFromCurve(FloatCurve=DataInterface() {propertyValues {Curve {keys [
         {arriveTangent=3.1235, interpMode=RCIM_Cubic, leaveTangent=3.1235, tangentMode=RCTM_User, time=0, value=0}
         {interpMode=RCIM_Cubic, tangentMode=RCTM_User, time=0.5, value=1}
         {arriveTangent=-3.4019, tangentMode=RCTM_User, time=1, value=0}
        ]}}})
   )
   &self:GetParentInfo(
    Hue Adjustment:NiagaraFloat=&RandomRangeFloat(Minimum=-0.2, Maximum=0.2)
    Normalize Color=true
    Parent Attribute Reader:NiagaraDataInterfaceParticleRead=$Emitter.Particle Attribute Reader
    ParentID:NiagaraInt32=$Particles.ParentPicker
    Value Adjustment:NiagaraFloat=&FloatFromCurve(FloatCurve=DataInterface() {propertyValues {Curve {keys [{arriveTangent=-27.5457, interpMode=RCIM_Cubic, leaveTangent=-27.5457, tangentMode=RCTM_User, time=0, value=5}, <at 0.5 1 RCIM_Cubic>]}}})
   ) {module=GetParentInfo001}
   &Drag(Drag=&RandomRangeFloat(Minimum=1, Maximum=2), Use Linear Drag=true, Use Rotational Drag=false)
   &GravityForce(Gravity=&RandomRangeVector(Minimum=<0 0 -50>, Maximum=<0 0 -25>))
   &PointAttractionForce(
    Attraction Strength=5
    Attractor Position:NiagaraPosition=$Particles.ParentPosition
    Kill Radius=1
    Kill Within Radius=false
    Module.Use Falloff=true
    UseAttractorOffset=false
   )
   &SolveForcesAndVelocity(Clamp Velocity=false, Limit Acceleration=false)
   /SetVariables(Particles.GlowColor=&MultiplyLinearColor(
     Linear Color=&Multiply_LinearColorByFloat(Linear Color=/Niagara/DynamicInputs/Add/Add_LinearColor(A:LinearColor=<rgba 1.5 1.2156 0.2516 1>, B:LinearColor=$Particles.Color), Float=0.6)
     Scale Factor=&MakeVector4(
      X=1
      Y=1
      Z=1
      W=&FloatFromCurve(
       FloatCurve=DataInterface() {propertyValues {Curve {
          defaultValue=0
          keys [
           {arriveTangent=-1.7417, interpMode=RCIM_Cubic, leaveTangent=-1.7417, tangentMode=RCTM_User, time=0, value=1}
           {arriveTangent=-0.2133, interpMode=RCIM_Cubic, leaveTangent=-0.2133, tangentMode=RCTM_User, time=1, value=0}
          ]
         }}}
       Scale Curve=&Multiply_Float(
        A=&FNTimeOfDayScalar(Night Scale=0.3, Morning Scale=0.4, Evening Scale=0.4)
        B=&MakeCustomFloatFromBool(
         Bool=&Boolean_OrOperation(Bool1=$User.IsFrontEnd, Bool2=&Boolean_OrOperation(Bool1=$User.bIsFrontEnd, Bool2=$User.bIsFrontEndPreview))
         True Float=0.75
         False Float=1
        )
       )
      )
     )
    )) {isSetParameters=true}
   /SetVariables(Particles.GlowColor=&MultiplyLinearColor(Linear Color=$Particles.Color, Scale Factor=<1 1 1 0.5>)) {enabled=false, isSetParameters=true}
  ]
  rendererClasses [&NiagaraSpriteRendererProperties]
  renderers [
   {rendererClass=&NiagaraSpriteRendererProperties}
   {rendererClass=&NiagaraSpriteRendererProperties, rendererIndex=1}
  ]
 }
}
summary {
 systemName=NS_Aura_FantasySparkle_Start
 emitters [
  {emitterName=Flare, rendererClasses [&NiagaraSpriteRendererProperties]}
  {emitterName=Dark_Glow, rendererClasses [&NiagaraSpriteRendererProperties]}
  {emitterName=Corona, rendererClasses [&NiagaraSpriteRendererProperties]}
  {emitterName=Sparks_Burst001, rendererClasses [&NiagaraSpriteRendererProperties]}
  {emitterName=Butterflies, rendererClasses [&NiagaraSpriteRendererProperties, &NiagaraMeshRendererProperties]}
  {emitterName=Trail, rendererClasses [&NiagaraSpriteRendererProperties]}
 ]
 userVariables [
  User.bIsFrontEnd {type=NiagaraBool}
  User.bIsFrontEndPreview {type=NiagaraBool}
  User.Interpolate Intensity {type=NiagaraFloat, defaultValue=2.5, description=""}
  User.IsFrontEnd {type=NiagaraBool}
  User.IsLargeBody {type=NiagaraBool}
  User.SkeletalMesh {
   type=NiagaraDataInterfaceSkeletalMesh
   defaultValue {
    struct=NiagaraExt_VariableValue_DataInterface
    value {dataInterface=&self:NiagaraDataInterfaceSkeletalMesh_1}
   }
  }
 ]
}
